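; BP_RE_VoidBurstOpen is a Blueprint that plays the opening flash effect of a void burst, using the rift spawn-in Niagara system.
; On begin play it checks authority and acts only on the remote branch, so the effect is purely cosmetic and skipped where it has authority.
; It spawns the flash slightly above the actor with a random rotation about one axis, then enlarges it through the Global Scale Niagara parameter.
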(event EventBeginPlay
  (Networking|SwitchHasAuthority
    (:Remote
      (bind _returnvalue (Niagara|SpawnSystematLocation (Variables|Default|GetFX_RiftSpawnIn_Flash) (+ (Transformation|GetActorLocation self) "0.000000,0.000000,100.000000") (Math|Rotator|MakeRotator 0.0 0.0 (Math|Random|RandomFloatinRange -1.8e+02 1.8e+02)) "1.000000,1.000000,1.000000" true true "None" true)) ; FlashFX
      (Niagara|SetNiagaraVariable(Float) _returnvalue "Global Scale" 2.0))))

(fn ConstructionScript)
